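# Base UIKit backplate widget: a single hit-test-invisible Image that draws a material-driven background with configurable brightness, corner radius and size.
# PreConstruct and OnInitialized apply those settings through setter functions.
# Setting the material rebuilds the dynamic instance, copying prior parameter overrides and setting blur from config, defaulting on.
# Enums drive material parameters, with corner radii taken from desktop or mobile presets, in pixels or relative units by type.
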WidgetBlueprint /FortUILibrary/Systems/Backplate/WBP_UIKit_Backplate_Base
graphs [
 EventGraph {nodes [
   n1 = Event PreConstruct()
   SetBackplateMaterial!(
    execute=[@n1.then, @n2.then]
    self
    Material Instance:Material Instance Object Reference=Get BackplateMaterial() -> (BackplateMaterial:Material Instance Object Reference)
   ) -> (then)
   > SetBrightness!(execute, self, Brightness:E_UI_BackplateBrightness Enum=Get Brightness() -> (Brightness:E_UI_BackplateBrightness Enum=NewEnumerator0)) -> (then)
   > SetCornerRadius!(
    execute
    self
    Corner Radius:E_UI_BackplateCornerRadius Enum=Get CornerRadius() -> (CornerRadius:E_UI_BackplateCornerRadius Enum=NewEnumerator0)
   ) -> (then)
   > SetBackplateSize!(execute, self, Size:Vector 2D Structure=Get BackplateSize() -> (BackplateSize:Vector 2D Structure)) -> (then)
   n2 = Event OnInitialized()
  ]}
 RefreshBackplateMID {nodes [
   RefreshBackplateMID() {is=title K2Node_FunctionEntry}
   > n1 = GetDynamicMaterial(self=Get Backplate() -> (Backplate:Image Object Reference))
   > n2 = Set NewMID(NewMID:Material Instance Dynamic Object Reference=@n1) -> (Output_Get:Material Instance Dynamic Object Reference)
   > Copy MID Parameter Overrides!(
    execute
    Target MID:Material Instance Dynamic Object Reference=@n2
    Source MID:Material Instance Dynamic Object Reference=Get BackplateMID() -> (BackplateMID:Material Instance Dynamic Object Reference)
   ) -> (then)
   > n3 = Set BackplateMID(BackplateMID:Material Instance Dynamic Object Reference=@n2) -> (Output_Get:Material Instance Dynamic Object Reference)
   n4 = GetConfigDataBool!(EntryName:String=EnableBlurForFrontEndWidgets) -> (bValueOut:bool, ReturnValue:bool)
   > SetScalarParameterValue(
    self=@n3
    ParameterName=UseBlur
    Value="To Float (Boolean)"(InBool=Select(Option 0:bool=true, Option 1:bool=@n4, Index=@n4.ReturnValue) -> (ReturnValue:bool))
   )
  ]}
 SetBackplateMaterial {nodes [
   n1 = SetBackplateMaterial!() -> (then, Material Instance:Material Instance Object Reference) {is=title K2Node_FunctionEntry}
   > n2 = Set BackplateMaterial(BackplateMaterial:Material Instance Object Reference=@n1) -> (Output_Get:Material Instance Object Reference)
   > SetBrushFromMaterial(self=Get Backplate(), Material=@n2)
   > RefreshBackplateMID() {is=title K2Node_CallFunction}
  ]}
 SetBackplateSize {nodes [
   n1 = SetBackplateSize!() -> (then, Size:Vector 2D Structure) {is=title K2Node_FunctionEntry}
   > SetDesiredSizeOverride!(execute, self:Image Object Reference=Get Backplate(), DesiredSize:Vector 2D Structure=@n1) -> (then)
  ]}
 SetBrightness {nodes [
   n1 = SetBrightness!() -> (then, Brightness:E_UI_BackplateBrightness Enum=NewEnumerator0) {is=title K2Node_FunctionEntry}
   > n2 = Set Brightness(Brightness:E_UI_BackplateBrightness Enum=@n1) -> (Output_Get:E_UI_BackplateBrightness Enum=NewEnumerator0)
   > SetScalarParameterValue(
    self=Get BackplateMID()
    ParameterName=Backplate Switch
    Value=Select!(NewEnumerator0:float=0, NewEnumerator1:float=1, NewEnumerator2:float=2, NewEnumerator3:float=3, Index:E_UI_BackplateBrightness Enum=@n2) -> (ReturnValue:float)
   )
  ]}
 SetCornerRadius {nodes [
   n1 = SetCornerRadius!() -> (then, Corner Radius:E_UI_BackplateCornerRadius Enum=NewEnumerator0) {is=title K2Node_FunctionEntry}
   > Set CornerRadius(CornerRadius:E_UI_BackplateCornerRadius Enum=@n1) -> (Output_Get:E_UI_BackplateCornerRadius Enum=NewEnumerator0)
   n2 = Get CornerRadius() -> (CornerRadius=NewEnumerator0)
   n3 = Get BackplateMID()
   > SetScalarParameterValue(
    self=@n3
    ParameterName=Radius in Pixels
    Value="To Float (Boolean)"(InBool=Select!(
      NewEnumerator0:bool=true
      NewEnumerator1:bool=true
      NewEnumerator3:bool=false
      NewEnumerator6:bool=true
      NewEnumerator4:bool=false
      NewEnumerator7:bool=false
      Index:E_UI_BackplateCornerRadius Enum=@n2
     ) -> (ReturnValue:bool))
   )
   n4 = Break S UI Backplate Corner Radius Values!(S_UI_BackplateCornerRadiusValues:"S UI Backplate Corner Radius Values Structure (by ref)"=Get Corners() -> (Corners:S UI Backplate Corner Radius Values Structure)) -> (
    Backplate_11_#1:double=~
    InteractiveComponent_12_#2:double=~
    Button_13_#3:double=~
    Tab_14_#4:double=~
    Circle_15_#5:double=~
   ) {is=title K2Node_BreakStruct}
   n5 = Get CornerRadius() -> (CornerRadius=NewEnumerator0)
   n6 = Break S UI Backplate Corner Radius Values(S_UI_BackplateCornerRadiusValues=Get MobileCorners() -> (MobileCorners:S UI Backplate Corner Radius Values Structure)) -> (
    Backplate_11_#1=~
    InteractiveComponent_12_#2=~
    Button_13_#3=~
    Tab_14_#4=~
    Circle_15_#5=~
   )
   n7 = Select!(
    Option 0:double=Select!(
     NewEnumerator0:double=@n4
     NewEnumerator1:double=@n4.InteractiveComponent_12_#2
     NewEnumerator3:double=@n4.Button_13_#3
     NewEnumerator6:double=@n4.Tab_14_#4
     NewEnumerator4:double=@n4.Circle_15_#5
     NewEnumerator7:double=0
     Index:E_UI_BackplateCornerRadius Enum=@n5
    ) -> (ReturnValue)
    Option 1:double=Select(
     NewEnumerator0=@n6
     NewEnumerator1=@n6.InteractiveComponent_12_#2
     NewEnumerator3=@n6.Button_13_#3
     NewEnumerator6=@n6.Tab_14_#4
     NewEnumerator4=@n6.Circle_15_#5
     NewEnumerator7=0
     Index=@n5
    )
    Index:bool=IsMobileGame()
   ) -> (ReturnValue:double)
   > SetVectorParameterValue(
    self=@n3
    ParameterName=Corner Radius
    Value=Make Linear Color!(R:float=@n7, G:float=@n7, B:float=@n7, A:float=@n7) -> (LinearColor:Linear Color Structure) {is=title K2Node_MakeStruct}
   )
  ]}
]
variables [
 BackplateMaterial {type=MaterialInstance, default=/FortUILibrary/Systems/Backplate/MaterialInstances/MI_UIKit_Backplate}
 BackplateMID {type=MaterialInstanceDynamic}
 BackplateSize {type=Vector2D, default=<0 0>}
 Brightness {type=E_UI_BackplateBrightness, default=Light}
 CornerRadius {type=E_UI_BackplateCornerRadius, default=Backplate}
 Corners {type=S_UI_BackplateCornerRadiusValues, default {backplate=10, button=0.3, circle=0.5, interactive Component=20, tab=17}}
 MobileCorners {type=S_UI_BackplateCornerRadiusValues, default {backplate=25, button=0.3, circle=0.5, interactive Component=20, tab=24}}
]
widgets {
 info {inheritedWidgetCount=0, parentClass=&UserWidget, rootWidgetClass=&Image, widgetCount=1}
 widgets [Backplate {
   widgetClassPath=&Image
   bIsVariable=true
   widget=Backplate
   properties {
    visibility=HitTestInvisible
    brush {
     bIsDynamicallyLoaded=false
     drawAs=Image
     imageType=NoImage
     mirroring=NoMirror
     resourceName=None
     resourceObject=null
     tiling=NoTile
     imageSize=<8 8>
     margin=<ltrb 0 0 0 0>
     outlineSettings {
      bUseBrushTransparency=false
      roundingType=HalfHeightRadius
      width=0
      color {colorUseRule=UseColor_Specified, specifiedColor=<rgba 0 0 0 0>}
      cornerRadii=<0 0 0 0>
     }
     tintColor {colorUseRule=UseColor_Specified, specifiedColor=<rgba 1 1 1 1>}
     uVRegion {bIsValid=false, max=<0 0>, min=<0 0>}
    }
   }
  }]
}
